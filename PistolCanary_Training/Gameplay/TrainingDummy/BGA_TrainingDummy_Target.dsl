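; Shootable training dummy target for the PistolCanary event that scores leaderboard points, drops loot and respawns.
; BeginPlay activates it and binds damage and death events; server-side hits credit the player, then destroy the actor.
; Death broadcasts a verb, spawns health, ammo and conditional weapon loot, and, unless a calendar event is active, requests delayed respawn from the pre-event manager.
; Active's OnRep drives material fading, collision, movement and client-side facing toward the closest player.
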(event Custom|ActivateTarget
  ;[0 0.06819 0.5061 0.5] Activate State and bind event to deactivate state
  (CallFunction|SpawnFXAtLocation)
  (Variables|Default|SetActive true)
  (Variables|Actor|SetbCanBeDamaged true)
  (Variables|Default|SetShot false)
  (EventDispatchers|BindEventtoOnDamaged (delegate Custom|OnDamaged_Event))
  (EventDispatchers|BindEventtoOnDied (EventDispatchers|CreateEvent :!event Custom|OnDiedEvent)))

(event Custom|OnDamaged_Event
  (params (DamagedActor :type Actor%20Object%20Reference) (Damage :type Float%20%28single-precision%29) (InstigatedBy :type Controller%20Object%20Reference) (DamageCauser :type Actor%20Object%20Reference) (HitLocation :type Vector) (FHitComponent :type Primitive%20Component%20Object%20Reference) (BoneName :type Name) (Momentum :type Vector))
  ;[0 0.06819 0.5061 0.5] Activate State and bind event to deactivate state
  (Variables|Default|SetInstigatedActor InstigatedBy)
  (if (Networking|IsDedicatedServer)
    (bind _asfort_player_pawn (Utilities|Casting|CastToFortPlayerPawn (Actor|GetOwner DamageCauser))
      (:then
        (CallFunction|AddPointstoLeaderBoard _asfort_player_pawn))
      (:CastFailed
        (bind _asfort_player_pawn_2 (Utilities|Casting|CastToFortPlayerPawn DamageCauser)
          (:then
            (CallFunction|AddPointstoLeaderBoard _asfort_player_pawn_2))
          (:CastFailed
            _))))
    (Variables|Default|SetShot true)
    (CallFunction|DeactivateTarget)
    (else
      _)))

(event EventBeginPlay
  ;[0 0.06819 0.5061 0.5] Begin Play Toggle State
  (|CallParentFunction)
  (CallFunction|ActivateTarget)
  (Variables|Default|SetCanRespawn true)
  (AimAssist|RegisterAimAssistTarget 0 (Variables|Default|GetCapsuleCollision)))

(event Custom|DeactivateTarget
  ;[0 0.06819 0.5061 0.5] Deactivate state
  (EventDispatchers|UnbindEventfromOnDamaged (EventDispatchers|CreateEvent :!event Custom|OnDamaged_Event))
  (Variables|Default|SetActive false)
  (Variables|Actor|SetbCanBeDamaged false)
  (Actor|DestroyActor))

(event Custom|ToggleState
  (params (Active :type Boolean))
  ;[0 0.06819 0.5061 0.5] Toggles on and off state
  (Rendering|Material|SetMaterial (Variables|Default|Getshooting_target_base_01) 0 "/PistolCanary_Training/Gameplay/TrainingDummy/Materials/M_TrainingDummy_Target_Inst.M_TrainingDummy_Target_Inst")
  (if Active
    (Rendering|SetActorHiddenInGame false)
    (bind (_direction _time) (Timeline|Fade PlayFromStart)
      (:Update
        (Rendering|Material|SetScalarParameterforCustomPrimitiveData (Variables|Default|Getshooting_target_base_01) "SpawnInOut" _time))
      (:Finished
        (Rendering|Material|SetMaterial (Variables|Default|Getshooting_target_base_01) 0 "/PistolCanary_Training/Gameplay/TrainingDummy/Materials/M_TrainingDummy_Target_Opaque.M_TrainingDummy_Target_Opaque")
        (if Active
          (Components|Activation|Activate (Variables|Default|GetInterpToMovement) true)
          (Collision|SetCollisionEnabled (Variables|Default|GetCapsuleCollision) "QueryOnly")
          (CallFunction|FacePlayer)
          (else
            (Rendering|SetActorHiddenInGame true)
            (Components|Activation|Deactivate (Variables|Default|GetInterpToMovement))))))
    (else
      (Collision|SetCollisionEnabled (Variables|Default|GetCapsuleCollision) "NoCollision")
      (Utilities|Time|ClearandInvalidateTimerbyHandle (Variables|Default|GetTimer_LookAtPlayer))
      (if (not (Variables|Default|GetShot))
        (Timeline|Fade ReverseFromEnd)
        (else
          (CallFunction|SpawnFXAtLocation))))))

(event Custom|SpawnFXAtLocation
  (Niagara|SpawnSystematLocation "/PistolCanary_Training/Gameplay/TrainingDummy/FX/NS_TrainingDummy_Target_Hit_FX.NS_TrainingDummy_Target_Hit_FX" (Transformation|GetWorldLocation (Variables|Default|GetFX_Location))))

(event Custom|ForceSelfDestruct
  (Variables|Default|SetUpdateDestruction (+ (Variables|Default|GetUpdateDestruction) 1)))

(event Custom|FacePlayer
  (bind _returnvalue (Transformation|GetActorLocation))
  (bind _returnvalue_2 (Fortnite|GetClosestFortPlayerPawn _returnvalue))
  (bind _shooting_target_base_01 (Variables|Default|Getshooting_target_base_01))
  ;[0 0.06819 0.5061 0.5] Toggles on and off state
  (Networking|SwitchHasAuthority
    (:Remote
      (Utilities|IsValid _returnvalue_2
        (:Is%20Valid
          (bind _returnvalue_3 (Animation|Utilities|GetFacingTargetRotation _shooting_target_base_01 (Variables|/Script/Engine.Actor|GetRootComponent _returnvalue_2)))
          (Transformation|SetRelativeRotation _shooting_target_base_01 (Math|Rotator|CombineRotators _returnvalue_3 "0.000000,-90.000000,0.000000") false false))))))

(event Custom|OnDiedEvent
  (params (DamagedActor :type Actor%20Object%20Reference) (Damage :type Float%20%28single-precision%29) (InstigatedBy :type Controller%20Object%20Reference) (DamageCauser :type Actor%20Object%20Reference) (HitLocation :type Vector) (FHitComponent :type Primitive%20Component%20Object%20Reference) (BoneName :type Name) (Momentum :type Vector))
  (bind _respawninterval (Variables|Default|GetRespawnInterval))
  (bind _returnvalue (ScalableFloat|GetValueatLevel _respawninterval))
  (bind _instigatedactor (Variables|Default|GetInstigatedActor)
    (:then
      (CustomVerb|BroadcastCustomVerbMessage (CustomVerb|MakeCustomVerbMessage _instigatedactor "(TagName=\"CustomVerb.PistolCanary.TargetDestroyed\")")))
    (:else
      _))
  (if (Variables|Default|GetCanRespawn)
    (bind _itempickupsspawned (Class|AthenaFunctionLibrary|SpawnLootfromTable (Utilities|Array|RandomArrayItem (Variables|Default|GetHealthPkg)) -1 -1 (Transformation|GetActorLocation) (Transformation|GetActorUpVector) 1 true true))
    (for _array_element _itempickupsspawned
      (Actor|SetLifeSpan _array_element (ScalableFloat|GetValueatLevel (Variables|Default|GetLootLifeSpan))))
    (bind _itempickupsspawned_2 (Class|AthenaFunctionLibrary|SpawnLootfromTable "Loot_Spire_Ammo" -1 -1 (Transformation|GetActorLocation) (Transformation|GetActorUpVector) 3 true true))
    (for _ _itempickupsspawned_2
      (Actor|SetLifeSpan (ScalableFloat|GetValueatLevel (Variables|Default|GetLootLifeSpan))))
    (bind _biseventactive (CalendarEvents|GetCalendarEventInformation (Variables|Default|GetWeaponSpawnEvent)))
    (if (not _biseventactive)
      (if (CallFunction|CheckifWeaponsActive)
        (bind _itempickupsspawned_3 (Class|AthenaFunctionLibrary|SpawnLootfromTable (Utilities|Array|RandomArrayItem (Variables|Default|GetWeaponPkg)) -1 -1 (Transformation|GetActorLocation) (Transformation|GetActorUpVector) 1 true true))
        (for _array_element_2 _itempickupsspawned_3
          (Actor|SetLifeSpan _array_element_2 _returnvalue))
        (else
          _))
      (Actor|SetLifeSpan (+ _returnvalue 2.0))
      (bind _returnvalue_2 (Actor|GetActorOfClass "/PistolCanary_Training/Gameplay/BP_PistolCanary_EventZone_PreEvent_GameplayManager.BP_PistolCanary_EventZone_PreEvent_GameplayManager_C"))
      (Utilities|FlowControl|Delay (ScalableFloat|GetValueatLevel (Variables|Default|GetRespawnInterval)))
      (if (Variables|Default|GetCanRespawn)
        (Class|BPPistolCanaryEventZonePreEventGameplayManager|RespawnDummy _returnvalue_2 (Utilities|GetClass self) (Transformation|GetActorTransform))))))

(fn AddPointsToLeaderBoard
  (params (PlayerState :type Pawn%20Object%20Reference))
  (bind _asfort_player_state (Utilities|Casting|CastToFortPlayerState (Variables|/Script/Engine.Pawn|GetPlayerState PlayerState)))
  (Class|SpecialEventMeshLeaderboardManager|RequestAddPlayerScoretoLeaderboard (Class|SpecialEventMeshLeaderboardManager|GetMeshLeaderboardManager false) _asfort_player_state (ScalableFloat|GetValueatLevel (Variables|Default|GetHF_DummyDeathScore))))

(fn CheckIfWeaponsActive
  (bind _outactors (Actor|GetAllActorsOfClass "/PistolCanary_Training/Gameplay/BP_PistolCanary_EventZoneManager.BP_PistolCanary_EventZoneManager_C"))
  (for _array_element _outactors
    (Utilities|IsValid _array_element
      (:Is%20Valid
        (return (Variables|/PistolCanary_Training/Gameplay/BP_PistolCanary_EventZoneManager.BP_PistolCanary_EventZoneManager_C|GetInventoryPhase2 _array_element))))))

(fn GetTargetOwner
  (return 0))

(fn IsValidAutoFireTarget
  (params (Attacker :type Actor%20Object%20Reference))
  (return (ScalableFloat|GetValueasBool (Variables|Default|GetAllowAutoFire))))

(fn OnRep_Active
  (CallFunction|ToggleState (Variables|Default|GetActive)))

(fn OnRep_UpdateDestruction
  (Collision|SetCollisionEnabled (Variables|Default|GetCapsuleCollision) "NoCollision")
  (Utilities|Time|ClearandInvalidateTimerbyHandle (Variables|Default|GetTimer_LookAtPlayer))
  (Rendering|SetActorHiddenInGame true)
  (Components|Activation|Deactivate (Variables|Default|GetInterpToMovement))
  (if (not (Variables|Default|GetShot))
    (CallFunction|SpawnFXAtLocation)))

(fn ShouldAlliesAutoFireOnDeployedTarget
  (return false))

(fn ShouldAutoFireOnSelfDeployedTarget
  (return false))

(fn ConstructionScript
  (Components|Movement|SetUpdatedComponent (Variables|Default|GetInterpToMovement) (Variables|Default|Getshooting_target_base_01)))
