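# Unlit masked Niagara particle material for the training dummy's digital cube effect, enabled for mesh and sprite particles.
# A Fresnel rim on vertex normals, lifted by a small floor, multiplies particle color and feeds MF_FX_EmissiveAdjustments for glow.
# Particle alpha drives DitherTemporalAA for dithered fades, with an If bypassing the dither when alpha is nearly full.
# A shading path switch adds the same default constant to color on every path.
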Material /PistolCanary_Training/Gameplay/TrainingDummy/FX/Materials/M_TrainingDummy_Digital_Cube
graph {
 properties {
  bAutomaticallySetUsageInEditor=false
  bUsedWithNiagaraMeshParticles=true
  bUsedWithNiagaraSprites=true
  blendMode=BLEND_Masked
  shadingModel=MSM_Unlit
  usedShadingModels=Unlit
 }
 nodes [
  n1 = ParticleColor()
  n2 = MaterialFunctionCall(Alpha Threshold=@n1.A) {materialFunction=&DitherTemporalAA}
  If(A=@n1.A, "A > B"=Constant() {r=1}, "A == B"=@n2.Result, "A < B"=@n2.Result) {constB=0.99}
  n3 = Fresnel(Normal=VertexNormalWS()) {exponent=3}
  n4 = Constant()
  n5 = ShadingPathSwitch(Default=@n4, Deferred=@n4, Forward=@n4, Mobile=@n4)
  MaterialFunctionCall(Emissive=Multiply(A=Add(A=@n3) {constB=0.2}, B=Add(A=@n1.RGB, B=@n5))) {materialFunction=&MF_FX_EmissiveAdjustments}
 ]
}
